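# Physics asset for the medium male Durrburger Worker player body, providing collision bodies and joints for the outfit's dynamic apron.
# Left and right apron chains each link two box-shaped bones to the pelvis, with narrower twist limits than swing.
# Capsules on the pelvis and thighs form the body volumes, with free-rotating thigh-to-pelvis joints.
# The left side is asymmetric, with extra apron boxes and a longer thigh capsule.
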PhysicsAsset /Game/Characters/Player/Male/Medium/Bodies/M_MED_DurrburgerWorker/Meshes/M_MED_DurrburgerWorker_PhysicsAsset
bodies {
 dyn_apron_01_l [
  {shapeType=Box, extentX=7.5, extentY=3.5, extentZ=7.5, shapeName=A, center=<-3.7912 0.7183 -0.3352>, rotation=<rot 0 -7.5686 0>}
  {shapeType=Box, extentX=5.5, extentY=3.5, extentZ=4, shapeName=Out, center=<-10.8247 1.4614 -4.2566>, rotation=<rot 1.8314 -4.0418 24.3372>}
 ]
 dyn_apron_01_r [{shapeType=Box, extentX=7.5, extentY=2, extentZ=7.5, center=<3.8194 -0.5068 0>, rotation=<rot 0 -7.5678 0>}]
 dyn_apron_l [
  {shapeType=Box, extentX=7.5, extentY=2, extentZ=7.5, shapeName=A, center=<-3.8194 0.5068 0>, rotation=<rot 0 -7.5678 0>}
  {shapeType=Box, extentX=7.5, extentY=2, extentZ=2, shapeName=B, center=<-4.1262 1.4252 6.1633>, rotation=<rot 0 -7.5687 0>}
 ]
 dyn_apron_r [{shapeType=Box, extentX=7.5, extentY=2, extentZ=7.5, center=<3.8194 -0.5068 0>, rotation=<rot 0 -7.5678 0>}]
 pelvis [{shapeType=Capsule, length=0, radius=14.25, center=<-0.0001 -1.7396 0>, rotation=<rot 0 0 90>}]
 thigh_l [{shapeType=Capsule, length=35.75, radius=10, center=<-11.6708 2.0617 1.2213>, rotation=<rot 90 0 4.7419>}]
 thigh_r [{shapeType=Capsule, length=22.75, radius=10, center=<11.7051 -1.3451 0.0003>, rotation=<rot 90 0 3.2161>}]
}
constraints [
 {bone1Name=dyn_apron_01_l, bone2Name=dyn_apron_l, swing1LimitDegrees=45, swing2LimitDegrees=45, twistLimitDegrees=5}
 {bone1Name=dyn_apron_01_r, bone2Name=dyn_apron_r, swing1LimitDegrees=45, swing2LimitDegrees=45, twistLimitDegrees=5}
 {
  bone1Name=thigh_l
  bone2Name=pelvis
  swing1LimitDegrees=45
  swing1Motion=Free
  swing2LimitDegrees=45
  swing2Motion=Free
  twistLimitDegrees=45
  twistMotion=Free
 }
 {
  bone1Name=thigh_r
  bone2Name=pelvis
  swing1LimitDegrees=45
  swing1Motion=Free
  swing2LimitDegrees=45
  swing2Motion=Free
  twistLimitDegrees=45
  twistMotion=Free
 }
 {bone1Name=dyn_apron_l, bone2Name=pelvis, swing1LimitDegrees=45, swing2LimitDegrees=45, twistLimitDegrees=5}
 {bone1Name=dyn_apron_r, bone2Name=pelvis, swing1LimitDegrees=45, swing2LimitDegrees=45, twistLimitDegrees=5}
]
