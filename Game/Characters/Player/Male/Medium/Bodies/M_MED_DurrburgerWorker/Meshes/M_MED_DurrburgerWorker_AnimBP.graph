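# Post-process-style AnimBlueprint for the male medium Durrburger Worker body that adds secondary motion to the costume's apron.
# The live chain copies the attached parent's pose, then runs spring-based Anim Dynamics on the right and left thigh apron bones before output.
# Other apron experiments (constraints, dynamics chains, flap and bone-driven nodes, Pose Driver, RigidBody, sequence players) are not wired to the output.
# The event graph holds no logic.
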AnimBlueprint /Game/Characters/Player/Male/Medium/Bodies/M_MED_DurrburgerWorker/Meshes/M_MED_DurrburgerWorker_AnimBP
target_skeleton=&/M_MED_DurrburgerWorker_Skeleton
graphs [
 AnimGraph {nodes [
   n1 = "Constraint - dyn_apron_l (R)"(Alpha=1, ConstraintWeights_0=1) {
    alphaBoolBlend {blendOption=Linear}
    boneToModify {boneName=dyn_apron_l}
    constraintSetup [{transformType=Rotation, targetBone {boneName=thigh_l}}]
    constraintWeights [1]
   }
   n2 = "Constraint - dyn_apron_r (R)"(ComponentPose=@n1, Alpha=1, ConstraintWeights_0=1) {
    alphaBoolBlend {blendOption=Linear}
    boneToModify {boneName=dyn_apron_r}
    constraintSetup [{transformType=Rotation, targetBone {boneName=thigh_r}}]
    constraintWeights [1]
   }
   n3 = "Constraint - dyn_apron_m (RR)"!(ComponentPose=@n2, Alpha=1, ConstraintWeights_0=0.5, ConstraintWeights_1:float=0.5) -> (Pose) {
    alphaBoolBlend {blendOption=Linear}
    boneToModify {boneName=dyn_apron_m}
    constraintSetup [{transformType=Rotation, targetBone {boneName=thigh_r}}, {transformType=Rotation, targetBone {boneName=thigh_l}}]
    constraintWeights [1,1]
   }
   n4 = "Anim Dynamics - Chain: dyn_apron_l -> dyn_apron_01_l"(ComponentPose=@n3, Alpha=1, ExternalForce="0,0,0") {
    angularDampingOverride=0.999
    angularSpringConstant=300
    bAngularSpring=true
    bChain=true
    bGravityOverrideInSimSpace=true
    bOverrideAngularDamping=true
    bUseSphericalLimits=true
    numSolverIterationsPostUpdate=4
    numSolverIterationsPreUpdate=16
    alphaBoolBlend {blendOption=Linear}
    boundBone {boneName=dyn_apron_l}
    chainEnd {boneName=dyn_apron_01_l}
    physicsBodyDefinitions [{
      collisionType=CustomSphere
      sphereCollisionRadius=2
      boundBone {boneName=dyn_apron_l}
      boxExtents=<8 25 18>
      constraintSetup {angularTargetAxis=AxisZ, angularLimitsMax=<20 5 0>, angularLimitsMin=<-20 -5 -30>, angularTarget=<0 0 1>}
      localJointOffset=<-5 0 0>
     }]
    sphericalLimits [
     {limitRadius=11, drivingBone {boneName=thigh_l}, sphereLocalOffset=<0 0 0>}
     {limitRadius=10, drivingBone {boneName=thigh_l}, sphereLocalOffset=<-5 0 0>}
    ]
   }
   n5 = "Anim Dynamics - Chain: dyn_apron_m -> dyn_apron_01_m"(ComponentPose=@n4, Alpha=1, ExternalForce="0,0,0") {
    angularDampingOverride=0.999
    angularSpringConstant=300
    bAngularSpring=true
    bChain=true
    bGravityOverrideInSimSpace=true
    bOverrideAngularDamping=true
    numSolverIterationsPostUpdate=4
    numSolverIterationsPreUpdate=16
    alphaBoolBlend {blendOption=Linear}
    boundBone {boneName=dyn_apron_m}
    chainEnd {boneName=dyn_apron_01_m}
    physicsBodyDefinitions [{
      sphereCollisionRadius=0
      boundBone {boneName=dyn_apron_m}
      boxExtents=<8 25 18>
      constraintSetup {angularTargetAxis=AxisZ, angularLimitsMax=<20 5 5>, angularLimitsMin=<-20 -5 -30>, angularTarget=<0 0 1>}
      localJointOffset=<-5 0 0>
     }]
   }
   "Anim Dynamics - Chain: dyn_apron_r -> dyn_apron_01_r"(ComponentPose=@n5, Alpha=1, ExternalForce="0,0,0") {
    angularDampingOverride=0.999
    angularSpringConstant=300
    bAngularSpring=true
    bChain=true
    bGravityOverrideInSimSpace=true
    bOverrideAngularDamping=true
    bUseSphericalLimits=true
    numSolverIterationsPostUpdate=4
    numSolverIterationsPreUpdate=16
    alphaBoolBlend {blendOption=Linear}
    boundBone {boneName=dyn_apron_r}
    chainEnd {boneName=dyn_apron_01_r}
    physicsBodyDefinitions [{
      collisionType=CustomSphere
      sphereCollisionRadius=2
      boundBone {boneName=dyn_apron_r}
      boxExtents=<8 25 18>
      constraintSetup {angularTargetAxis=AxisZ, angularLimitsMax=<20 5 0>, angularLimitsMin=<-20 -5 -30>, angularTarget=<0 0 1>}
      localJointOffset=<5 0 0>
     }]
    sphericalLimits [
     {limitRadius=11, drivingBone {boneName=thigh_r}, sphereLocalOffset=<0 0 -2>}
     {limitRadius=11, drivingBone {boneName=thigh_r}, sphereLocalOffset=<5 0 -2>}
    ]
   }
   n6 = Flap Node - Bone: dyn_apron_r!(ComponentPose:Component Space Pose Link Structure, Alpha:float=1) -> (Pose:Component Space Pose Link Structure) {
    is=title FortAnimGraphNode_Flap
    bUseClamp=true
    targetClampMax=-0.2
    targetClampMin=-3.14
    targetComponent=RotationZ
    alphaBoolBlend {blendOption=Linear}
    sourceDrives [{
      bUseRange=false
      drivingCurve=null
      multiplier=-1
      rangeMax=1
      rangeMin=-1
      remappedMax=1
      remappedMin=-1
      sourceComponent=RotationZ
      useQuaternion=true
      sourceBone {boneName=thigh_r}
     }]
    targetBone {boneName=dyn_apron_r}
   }
   "dyn_apron_r.rotateX = thigh_r.rotateX * 0.5 - Bone Driven Controller"!(ComponentPose:Component Space Pose Link Structure=@n6, Alpha:float=1) -> (Pose:Component Space Pose Link Structure) {
    is=title AnimGraphNode_BoneDrivenController
    bAffectTargetRotationX=true
    multiplier=0.5
    sourceComponent=RotationX
    alphaBoolBlend {blendOption=Linear}
    sourceBone {boneName=thigh_r}
    targetBone {boneName=dyn_apron_r}
   }
   n7 = Flap Node - Bone: dyn_apron_l!(ComponentPose:Component Space Pose Link Structure, Alpha:float=1) -> (Pose:Component Space Pose Link Structure) {
    is=title FortAnimGraphNode_Flap
    bUseClamp=true
    targetClampMin=-3.14
    targetComponent=RotationZ
    alphaBoolBlend {blendOption=Linear}
    sourceDrives [{
      bUseRange=false
      drivingCurve=null
      multiplier=-1
      rangeMax=1
      rangeMin=-1
      remappedMax=1
      remappedMin=-1
      sourceComponent=RotationZ
      useQuaternion=true
      sourceBone {boneName=thigh_l}
     }]
    targetBone {boneName=dyn_apron_l}
   }
   "dyn_apron_l.rotateX = thigh_l.rotateX * -0.5 - Bone Driven Controller"!(ComponentPose:Component Space Pose Link Structure=@n7, Alpha:float=1) -> (Pose:Component Space Pose Link Structure) {
    is=title AnimGraphNode_BoneDrivenController
    bAffectTargetRotationX=true
    multiplier=-0.5
    sourceComponent=RotationX
    alphaBoolBlend {blendOption=Linear}
    sourceBone {boneName=thigh_l}
    targetBone {boneName=dyn_apron_l}
   }
   n8 = "dyn_apron_l.rotateZ = thigh_l.rotateZ - Bone Driven Controller"!(ComponentPose:Component Space Pose Link Structure, Alpha:float=1) -> (Pose:Component Space Pose Link Structure) {
    is=title AnimGraphNode_BoneDrivenController
    bAffectTargetRotationZ=true
    sourceComponent=RotationZ
    alphaBoolBlend {blendOption=Linear}
    sourceBone {boneName=thigh_l}
    targetBone {boneName=dyn_apron_l}
   }
   n9 = "dyn_apron_l.rotateY = thigh_l.rotateY - Bone Driven Controller"!(ComponentPose:Component Space Pose Link Structure=@n8, Alpha:float=1) -> (Pose:Component Space Pose Link Structure) {
    is=title AnimGraphNode_BoneDrivenController
    bAffectTargetRotationY=true
    sourceComponent=RotationY
    alphaBoolBlend {blendOption=Linear}
    sourceBone {boneName=thigh_l}
    targetBone {boneName=dyn_apron_l}
   }
   n10 = "dyn_apron_l.rotateX = thigh_l.rotateX * -0.5 - Bone Driven Controller"(ComponentPose=@n9, Alpha=1) {
    bAffectTargetRotationX=true
    multiplier=-0.5
    sourceComponent=RotationX
    alphaBoolBlend {blendOption=Linear}
    sourceBone {boneName=thigh_l}
    targetBone {boneName=dyn_apron_l}
   }
   n11 = "dyn_apron_r.rotateZ = thigh_r.rotateZ - Bone Driven Controller"!(ComponentPose:Component Space Pose Link Structure=@n10, Alpha:float=1) -> (Pose:Component Space Pose Link Structure) {
    is=title AnimGraphNode_BoneDrivenController
    bAffectTargetRotationZ=true
    sourceComponent=RotationZ
    alphaBoolBlend {blendOption=Linear}
    sourceBone {boneName=thigh_r}
    targetBone {boneName=dyn_apron_r}
   }
   n12 = "dyn_apron_r.rotateY = thigh_r.rotateY - Bone Driven Controller"!(ComponentPose:Component Space Pose Link Structure=@n11, Alpha:float=1) -> (Pose:Component Space Pose Link Structure) {
    is=title AnimGraphNode_BoneDrivenController
    bAffectTargetRotationY=true
    sourceComponent=RotationY
    alphaBoolBlend {blendOption=Linear}
    sourceBone {boneName=thigh_r}
    targetBone {boneName=dyn_apron_r}
   }
   "dyn_apron_r.rotateX = thigh_r.rotateX * -0.5 - Bone Driven Controller"!(ComponentPose:Component Space Pose Link Structure=@n12, Alpha:float=1) -> (Pose:Component Space Pose Link Structure) {
    is=title AnimGraphNode_BoneDrivenController
    bAffectTargetRotationX=true
    multiplier=-0.5
    sourceComponent=RotationX
    alphaBoolBlend {blendOption=Linear}
    sourceBone {boneName=thigh_r}
    targetBone {boneName=dyn_apron_r}
   }
   "dyn_apron_r.rotateZ = thigh_r.rotateZ - Bone Driven Controller"(Alpha=1) {
    bAffectTargetRotationZ=true
    sourceComponent=RotationZ
    alphaBoolBlend {blendOption=Linear}
    sourceBone {boneName=thigh_r}
    targetBone {boneName=dyn_apron_r}
   }
   Copy Bone - Source Bone: thigh_r - Target Bone: dyn_apron_r!(
    ComponentPose:Component Space Pose Link Structure=Copy Bone - Source Bone: thigh_l - Target Bone: dyn_apron_l!(
     ComponentPose:Component Space Pose Link Structure
     Alpha:float=1
     bCopyTranslation:bool=False
     bCopyRotation:bool=true
     bCopyScale:bool=False
    ) -> (Pose:Component Space Pose Link Structure) {is=title AnimGraphNode_CopyBone, alphaBoolBlend {blendOption=Linear}, sourceBone {boneName=thigh_l}, targetBone {boneName=dyn_apron_l}}
    Alpha:float=1
    bCopyTranslation:bool=False
    bCopyRotation:bool=true
    bCopyScale:bool=False
   ) -> (Pose:Component Space Pose Link Structure) {is=title AnimGraphNode_CopyBone, alphaBoolBlend {blendOption=Linear}, sourceBone {boneName=thigh_r}, targetBone {boneName=dyn_apron_r}}
   Pose Driver - Source: thigh_l!(SourcePose:Pose Link Structure) -> (Pose:Pose Link Structure) {
    is=title AnimGraphNode_PoseDriver
    poseAsset=/Game/Characters/Player/Male/Medium/Bodies/M_MED_DurrburgerWorker/Animation/DurrburgerWorker_Skirt_PoseAsset
    evalSpaceBone {boneName=pelvis}
    onlyDriveBones [{boneName=dyn_apron_l}, {boneName=dyn_apron_01_l}, {boneName=dyn_apron_r}, {boneName=dyn_apron_01_r}]
    poseTargets [
     {
      bApplyCustomCurve=false
      bIsHidden=false
      distanceMethod=DefaultMethod
      drivenName=Base
      functionType=DefaultFunction
      targetScale=0.4623
      boneTransforms [{targetRotation=<rot 0 0 0>, targetTranslation=<0 0 0>}, {targetRotation=<rot 0 0 0>, targetTranslation=<0 0 0>}]
      customCurve {defaultValue=3.4028234663852886e+38, postInfinityExtrap=RCCE_Constant, preInfinityExtrap=RCCE_Constant, keys []}
      targetRotation=<rot 0 0 0>
     }
     {
      bApplyCustomCurve=false
      bIsHidden=false
      distanceMethod=DefaultMethod
      drivenName=Idle
      functionType=DefaultFunction
      targetScale=1
      boneTransforms [
       {targetRotation=<rot 15.8319 -20.0848 9.0531>, targetTranslation=<-2.5693 -0.0109 -7.7501>}
       {targetRotation=<rot 17.0245 -179.212 16.7827>, targetTranslation=<-2.5691 -0.0109 7.7501>}
      ]
      customCurve {defaultValue=3.4028234663852886e+38, postInfinityExtrap=RCCE_Constant, preInfinityExtrap=RCCE_Constant, keys []}
      targetRotation=<rot 0 0 0>
     }
     {
      bApplyCustomCurve=false
      bIsHidden=false
      distanceMethod=DefaultMethod
      drivenName=jog_l_leg_fwd
      functionType=DefaultFunction
      targetScale=0.3047
      boneTransforms [
       {targetRotation=<rot 15.533 -62.4297 13.9668>, targetTranslation=<-2.5689 -0.0119 -7.7495>}
       {targetRotation=<rot 3.1481 174.3445 -9.4085>, targetTranslation=<-2.5682 -0.01 7.7494>}
      ]
      customCurve {defaultValue=3.4028234663852886e+38, postInfinityExtrap=RCCE_Constant, preInfinityExtrap=RCCE_Constant, keys []}
      targetRotation=<rot 0 0 0>
     }
     {
      bApplyCustomCurve=false
      bIsHidden=false
      distanceMethod=DefaultMethod
      drivenName=jog_r_leg_fwd
      functionType=DefaultFunction
      targetScale=0.1933
      boneTransforms [
       {targetRotation=<rot 1.7049 -8.3842 -3.5014>, targetTranslation=<-2.5582 -0.0178 -7.7379>}
       {targetRotation=<rot 9.4121 86.8902 0.0135>, targetTranslation=<-2.5693 -0.0011 7.7422>}
      ]
      customCurve {defaultValue=3.4028234663852886e+38, postInfinityExtrap=RCCE_Constant, preInfinityExtrap=RCCE_Constant, keys []}
      targetRotation=<rot 0 0 0>
     }
     {
      bApplyCustomCurve=false
      bIsHidden=false
      distanceMethod=DefaultMethod
      drivenName=crouch
      functionType=DefaultFunction
      targetScale=0.1461
      boneTransforms [
       {targetRotation=<rot 24.9462 -105.9211 -4.2544>, targetTranslation=<-2.5693 -0.0109 -7.7501>}
       {targetRotation=<rot 20.9763 153.3156 -14.8998>, targetTranslation=<-2.5691 -0.0109 7.7501>}
      ]
      customCurve {defaultValue=3.4028234663852886e+38, postInfinityExtrap=RCCE_Constant, preInfinityExtrap=RCCE_Constant, keys []}
      targetRotation=<rot 0 0 0>
     }
     {
      bApplyCustomCurve=false
      bIsHidden=false
      distanceMethod=DefaultMethod
      drivenName=crouch_r_leg_fwd
      functionType=DefaultFunction
      targetScale=0.1933
      boneTransforms [
       {targetRotation=<rot -4.0153 -36.0408 8.5976>, targetTranslation=<-2.5672 -0.0132 -7.7453>}
       {targetRotation=<rot 0.0069 75.3079 9.0925>, targetTranslation=<-2.5673 -0.0083 7.7454>}
      ]
      customCurve {defaultValue=3.4028234663852886e+38, postInfinityExtrap=RCCE_Constant, preInfinityExtrap=RCCE_Constant, keys []}
      targetRotation=<rot 0 0 0>
     }
     {
      bApplyCustomCurve=false
      bIsHidden=false
      distanceMethod=DefaultMethod
      drivenName=crouch_l_leg_fwd
      functionType=DefaultFunction
      targetScale=0.1461
      boneTransforms [
       {targetRotation=<rot 12.5469 -107.2827 -13.3371>, targetTranslation=<-2.5689 -0.0118 -7.7489>}
       {targetRotation=<rot 6.1573 138.3454 -9.259>, targetTranslation=<-2.5684 -0.0101 7.7487>}
      ]
      customCurve {defaultValue=3.4028234663852886e+38, postInfinityExtrap=RCCE_Constant, preInfinityExtrap=RCCE_Constant, keys []}
      targetRotation=<rot 0 0 0>
     }
    ]
    rBFParams {distanceMethod=Euclidean, function=Linear, radius=464.392}
    sourceBones [{boneName=thigh_l}, {boneName=thigh_r}]
   }
   RigidBody() {
    overridePhysicsAsset=&/M_MED_DurrburgerWorker_PhysicsAsset
    alphaBoolBlend {blendOption=Linear}
   }
   n13 = Anim Dynamics - Bone: dyn_thigh_apron_r(ComponentPose=Local To Component(LocalPose=Copy Pose From Mesh() {bUseAttachedParent=true}), Alpha=1, ExternalForce="0,0,0") {
    angularDampingOverride=0.99
    angularSpringConstant=300
    bAngularSpring=true
    bGravityOverrideInSimSpace=true
    bOverrideAngularDamping=true
    bUsePlanarLimit=false
    numSolverIterationsPostUpdate=4
    numSolverIterationsPreUpdate=16
    alphaBoolBlend {blendOption=Linear}
    boundBone {boneName=dyn_thigh_apron_r}
    componentLinearAccScale=<0.5 0.5 0.5>
    componentLinearVelScale=<0.5 0.5 0.5>
    physicsBodyDefinitions [{
      collisionType=CustomSphere
      sphereCollisionRadius=2
      boundBone {boneName=dyn_thigh_apron_r}
      boxExtents=<8 25 18>
      constraintSetup {angularTargetAxis=AxisZ, angularLimitsMax=<0 0 0>, angularLimitsMin=<0 0 -30>, angularTarget=<0 0 1>}
      localJointOffset=<5 0 0>
     }]
    sphericalLimits [
     {limitRadius=11, drivingBone {boneName=thigh_r}, sphereLocalOffset=<0 0 -2>}
     {limitRadius=11, drivingBone {boneName=thigh_r}, sphereLocalOffset=<5 0 -2>}
    ]
   }
   n14 = Anim Dynamics - Bone: dyn_thigh_apron_l(ComponentPose=@n13, Alpha=1, ExternalForce="0,0,0") {
    angularDampingOverride=0.99
    angularSpringConstant=300
    bAngularSpring=true
    bGravityOverrideInSimSpace=true
    bOverrideAngularDamping=true
    bUsePlanarLimit=false
    numSolverIterationsPostUpdate=4
    numSolverIterationsPreUpdate=16
    alphaBoolBlend {blendOption=Linear}
    boundBone {boneName=dyn_thigh_apron_l}
    componentLinearAccScale=<0.5 0.5 0.5>
    componentLinearVelScale=<0.5 0.5 0.5>
    physicsBodyDefinitions [{
      collisionType=CustomSphere
      sphereCollisionRadius=2
      boundBone {boneName=dyn_thigh_apron_l}
      boxExtents=<8 25 18>
      constraintSetup {angularTargetAxis=AxisZ, angularLimitsMax=<0 0 0>, angularLimitsMin=<0 0 -30>, angularTarget=<0 0 1>}
      localJointOffset=<-5 0 0>
     }]
    sphericalLimits [
     {limitRadius=11, drivingBone {boneName=thigh_r}, sphereLocalOffset=<0 0 -2>}
     {limitRadius=11, drivingBone {boneName=thigh_r}, sphereLocalOffset=<5 0 -2>}
    ]
   }
   Output Pose(Result=Component To Local(ComponentPose=@n14)) {layerGroup=None}
   Sequence Player 'Running_around'() {playRate=0.2, sequence=/Game/Characters/Player/Male/Medium/Bodies/M_MED_DurrburgerWorker/Animation/Running_around, startPosition=1.5}
   Sequence Player 'PoseAsset_Skirt'() {playRate=0, sequence=/Game/Characters/Player/Male/Medium/Bodies/M_MED_DurrburgerWorker/Animation/PoseAsset_Skirt, startPosition=0.06}
   n15 = Flap Node - Bone: dyn_apron_01_l!(ComponentPose:Component Space Pose Link Structure, Alpha:float=1) -> (Pose:Component Space Pose Link Structure) {
    is=title FortAnimGraphNode_Flap
    targetClampMax=3
    targetComponent=RotationY
    alphaBoolBlend {blendOption=Linear}
    sourceDrives [{
      bUseRange=false
      drivingCurve=null
      multiplier=1
      rangeMax=1
      rangeMin=-1
      remappedMax=1
      remappedMin=-1
      sourceComponent=RotationY
      useQuaternion=false
      sourceBone {boneName=thigh_l}
     }]
    targetBone {boneName=dyn_apron_01_l}
   }
   Flap Node - Bone: dyn_apron_01_r!(ComponentPose:Component Space Pose Link Structure=@n15, Alpha:float=1) -> (Pose:Component Space Pose Link Structure) {
    is=title FortAnimGraphNode_Flap
    targetClampMax=3
    targetComponent=RotationY
    alphaBoolBlend {blendOption=Linear}
    sourceDrives [{
      bUseRange=false
      drivingCurve=null
      multiplier=1
      rangeMax=1
      rangeMin=-1
      remappedMax=1
      remappedMin=-1
      sourceComponent=RotationY
      useQuaternion=false
      sourceBone {boneName=thigh_r}
     }]
    targetBone {boneName=dyn_apron_01_r}
   }
   Flap Node - Bone: dyn_apron_m!(ComponentPose:Component Space Pose Link Structure, Alpha:float=1) -> (Pose:Component Space Pose Link Structure) {
    is=title FortAnimGraphNode_Flap
    targetClampMax=3
    targetComponent=RotationZ
    alphaBoolBlend {blendOption=Linear}
    sourceDrives [
     {
      bUseRange=false
      drivingCurve=null
      multiplier=-0.5
      rangeMax=1
      rangeMin=-1
      remappedMax=1
      remappedMin=-1
      sourceComponent=RotationZ
      useQuaternion=false
      sourceBone {boneName=thigh_l}
     }
     {
      bUseRange=false
      drivingCurve=null
      multiplier=-0.5
      rangeMax=1
      rangeMin=-1
      remappedMax=1
      remappedMin=-1
      sourceComponent=RotationZ
      useQuaternion=false
      sourceBone {boneName=thigh_r}
     }
    ]
    targetBone {boneName=dyn_apron_m}
   }
  ]}
 EventGraph {nodes [
   TryGetPawnOwner()
   Event BlueprintUpdateAnimation()
  ]}
]
variables []
