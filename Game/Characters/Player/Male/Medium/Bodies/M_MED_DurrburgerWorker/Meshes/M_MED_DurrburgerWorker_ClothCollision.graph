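# Physics asset for the medium male Durrburger Worker player body, named as the collision volume set for the character's cloth simulation.
# It approximates only the lower torso and upper legs, using a sphere and capsule on the pelvis, two capsules on spine_02, and one capsule per thigh.
# Two constraints link the thighs to the pelvis with swing and twist left free, while spine_02 has no constraint.
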PhysicsAsset /Game/Characters/Player/Male/Medium/Bodies/M_MED_DurrburgerWorker/Meshes/M_MED_DurrburgerWorker_ClothCollision
bodies {
 pelvis [{shapeType=Sphere, radius=15.25, center=<-8.6186 -1.283 0>}, {shapeType=Capsule, length=3.385, radius=17.8416, center=<-0.55 -4.4065 0>}]
 spine_02 [{shapeType=Capsule, length=15, radius=15, rotation=<rot 90 0 0>}, {shapeType=Capsule, length=5, radius=13.25, center=<2.8508 2.8216 0.0001>}]
 thigh_l [{shapeType=Capsule, length=26.5236, radius=10.3483, center=<-13.8183 2.0977 0.6062>, rotation=<rot 90 -90 -84.3116>}]
 thigh_r [{shapeType=Capsule, length=26.5236, radius=10.3483, center=<13.8183 -2.0977 -0.6062>, rotation=<rot 90 -90 -84.3116>}]
}
constraints [
 {
  bone1Name=thigh_l
  bone2Name=pelvis
  swing1LimitDegrees=45
  swing1Motion=Free
  swing2LimitDegrees=45
  swing2Motion=Free
  twistLimitDegrees=45
  twistMotion=Free
 }
 {
  bone1Name=thigh_r
  bone2Name=pelvis
  swing1LimitDegrees=45
  swing1Motion=Free
  swing2LimitDegrees=45
  swing2Motion=Free
  twistLimitDegrees=45
  twistMotion=Free
 }
]
